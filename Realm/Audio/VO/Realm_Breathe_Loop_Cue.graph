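# Looping breathing VO sound cue for the Realm content, playing a single wave on the in-game SFX sound class.
# One looping wave player feeds the output directly at reduced volume, spatialized through a dedicated attenuation asset.
# A concurrency override allows only one instance at a time, stopping the lowest-priority one when exceeded.
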SoundCue /Realm/Audio/VO/Realm_Breathe_Loop_Cue
root=WavePlayer_3
nodes [{id=WavePlayer_3, properties {bLooping=true, soundWaveAssetPtr=&/Realm_Breathe_Loop}}]
properties {
 attenuationSettings=&/Realm_Breathe_Loop_Att
 bOverrideConcurrency=true
 maxDistance=8000
 soundClassObject=&Fort_SFX_InGame
 volumeMultiplier=0.5
 concurrencyOverrides {maxCount=1, resolutionRule=StopLowestPriority}
}
